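# Sound cue that plays one Season 23 Battle Pass voice line by Amie, apparently for quest comms.
# A single Dialogue Player root node plays the dialogue wave with the Amie dialogue voice as speaker and no targets.
# The cue uses the shared quest comms attenuation, routes to the VO narrator sound class and boosts its volume.
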SoundCue /BattlePassS23/Audio/VO/Cues/AMIE_S23_REGS_03_04_140_Dialogue_Cue
root=DialoguePlayer_0
nodes [{
  id=DialoguePlayer_0
  properties {dialogueWaveParameter {
    dialogueWave=/BattlePassS23/Audio/VO/Waves/Amie/AMIE_S23_REGS_03_04_140_Dialogue
    context {speaker=/BattlePassS23/Audio/VO/Amie_DV, targets []}
   }}
 }]
properties {
 attenuationSettings=/BattlePassS23/Audio/VO/S23_Quest_Comms_Att
 maxDistance=2097152
 soundClassObject=/Game/Sounds/FortSoundClassesAndModes/Fort_VO_Narrator
 volumeMultiplier=2
}
